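# Sound cue for a shooting gallery pressure plate switching on, in the Figment Season 7 Athena quest audio.
# A single gallery activation wave plays through a Modulator node at the root.
# It uses gadget-activation attenuation and the in-game mission sound class, and borrows the rifle bullet impact concurrency set to limit simultaneous voices.
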SoundCue /Figment_S07/Athena/Sounds/Quests/Athena_ShootingGallery_PressurePlate_ON__Cue
root=Modulator_0
nodes [
 {id=Modulator_0, children [WavePlayer_0]}
 {id=WavePlayer_0, properties {soundWaveAssetPtr=&/FNBR_Gallery_Activate_01}}
]
properties {
 attenuationSettings=&Fort_GadgetActivate_Att
 maxDistance=4000
 soundClassObject=/Game/Sounds/FortSoundClassesAndModes/Fort_Mission_IngameSounds
 volumeMultiplier=1
 concurrencySet [&RifleBulletImpact_Conc]
}
