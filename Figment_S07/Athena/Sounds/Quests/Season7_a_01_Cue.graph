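# Sound cue for Season 7 quest audio in Figment, routed through the Fort_UI sound class.
# A Mixer plays two season7_a waves simultaneously at slightly different volumes, with a third input left empty.
# Overridden concurrency allows one instance at a time and stops the oldest when a new one starts.
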SoundCue /Figment_S07/Athena/Sounds/Quests/Season7_a_01_Cue
root=Mixer_0
nodes [
 {id=Mixer_0, children [WavePlayer_0,WavePlayer_2,null], properties {inputVolume [1.2,0.9,1]}}
 {id=WavePlayer_0, properties {soundWaveAssetPtr=&/season7_a_02}}
 {id=WavePlayer_2, properties {soundWaveAssetPtr=&/season7_a_01}}
]
properties {
 bOverrideConcurrency=true
 maxDistance=2097152
 soundClassObject=&Fort_UI
 volumeMultiplier=1
 concurrencyOverrides {maxCount=1, resolutionRule=StopOldest}
}
