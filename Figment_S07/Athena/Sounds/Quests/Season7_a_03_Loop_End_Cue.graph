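# Sound cue for a Figment Season 7 quest sound, named as the end of a loop, that plays a single wave once after a short delay.
# A Delay node feeds the season7_a_05 wave player, though the cue is named a_03, and its minimum delay is set higher than its maximum.
# It plays quietly in the gadgets sound class and overrides attenuation with short-range falloff while still referencing a supply drop marker attenuation asset.
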SoundCue /Figment_S07/Athena/Sounds/Quests/Season7_a_03_Loop_End_Cue
root=Delay_0
nodes [
 {id=Delay_0, children [WavePlayer_0], properties {delayMax=0.022, delayMin=0.22}}
 {id=WavePlayer_0, properties {soundWaveAssetPtr=&/season7_a_05}}
]
properties {
 attenuationSettings=/Figment_S07/Athena/Sounds/SupplyDrops/Fort_SupplyDrop_Marker_Close_Athena_Att
 bOverrideAttenuation=true
 maxDistance=800
 soundClassObject=&Fort_Gadgets
 volumeMultiplier=0.2
 attenuationOverrides {falloffDistance=700, attenuationShapeExtents {x=100}}
}
