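# Sound cue for a Season 7 quest sound in Figment, routed through the UI sound class with quest attenuation.
# A mixer layers the c_02 wave with a delayed, boosted c_01 wave and a quieter earthquake shake.
# The shake starts after a short delay and is shaped by an enveloper volume curve that swells and then fades.
# A concurrency override allows one instance, stopping the oldest.
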SoundCue /Figment_S07/Athena/Sounds/Quests/Season7_c_01_Cue
root=Mixer_0
nodes [
 {id=Delay_0, children [WavePlayer_1], properties {delayMax=0.47, delayMin=0.47}}
 {id=Delay_1, children [WavePlayer_5], properties {delayMax=0.3, delayMin=0.3}}
 {
  id=Enveloper_0
  children [Delay_1]
  properties {
   pitchCurve {editorCurveData {keys [{time=0}]}}
   volumeCurve {editorCurveData {keys [<at 0.4085 -0.0962>,<at 0.5 0>,<at 2.7 1>,<at 7.5 0.0495>]}}
  }
 }
 {id=Mixer_0, children [WavePlayer_3,Delay_0,Enveloper_0], properties {inputVolume [1,1.5,0.5]}}
 {id=WavePlayer_1, properties {soundWaveAssetPtr=&/season7_c_01}}
 {id=WavePlayer_3, properties {soundWaveAssetPtr=&/season7_c_02}}
 {id=WavePlayer_5, properties {soundWaveAssetPtr=/Figment_S07/Sounds/Fort_Impact_Sounds/Demolition/EarthQuake_Shakes_Med}}
]
properties {
 attenuationSettings=/Game/Sounds/Attenuation/Fort_Quest_Season7
 bOverrideConcurrency=true
 maxDistance=2400
 soundClassObject=&Fort_UI
 volumeMultiplier=1
 concurrencyOverrides {maxCount=1, resolutionRule=StopOldest}
}
