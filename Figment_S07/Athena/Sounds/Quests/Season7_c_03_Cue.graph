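# Season7_c_03_Cue is a quest sound cue for Figment Season 7 that plays one of seven season7_c_03 variations (a through g).
# A single Random node picks among the wave players with equal weight.
# It borrows trap sound settings, using the trap activation attenuation and the trap placement sound class.
# A concurrency override caps simultaneous instances at four and stops the oldest when the limit is exceeded.
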SoundCue /Figment_S07/Athena/Sounds/Quests/Season7_c_03_Cue
root=Random_0
nodes [
 {
  id=Random_0
  children [WavePlayer_0,WavePlayer_2,WavePlayer_4,WavePlayer_5,WavePlayer_6,WavePlayer_7,WavePlayer_8]
  properties {weights [1,1,1,1,1,1,1]}
 }
 {id=WavePlayer_0, properties {soundWaveAssetPtr=&/season7_c_03_a}}
 {id=WavePlayer_2, properties {soundWaveAssetPtr=&/season7_c_03_b}}
 {id=WavePlayer_4, properties {soundWaveAssetPtr=&/season7_c_03_c}}
 {id=WavePlayer_5, properties {soundWaveAssetPtr=&/season7_c_03_d}}
 {id=WavePlayer_6, properties {soundWaveAssetPtr=&/season7_c_03_e}}
 {id=WavePlayer_7, properties {soundWaveAssetPtr=&/season7_c_03_f}}
 {id=WavePlayer_8, properties {soundWaveAssetPtr=&/season7_c_03_g}}
]
properties {
 attenuationSettings=/Game/Sounds/Attenuation/Fort_Traps_Activate_Athena_Att
 bOverrideConcurrency=true
 maxDistance=3500
 soundClassObject=/Game/Sounds/FortSoundClassesAndModes/Fort_Traps_Placement
 volumeMultiplier=1
 concurrencyOverrides {maxCount=4, resolutionRule=StopOldest}
}
