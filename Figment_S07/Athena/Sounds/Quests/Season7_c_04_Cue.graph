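# Sound cue for a Season 7 quest in Figment, playing a medium earthquake shake sound.
# A single wave player feeds an enveloper whose volume curve swells over about a second, then fades to near silence by 7.5 seconds.
# It uses third-person quest attenuation but belongs to the UI sound class.
# Its concurrency override allows one instance, stopping the oldest when a new one starts.
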SoundCue /Figment_S07/Athena/Sounds/Quests/Season7_c_04_Cue
root=Enveloper_0
nodes [
 {
  id=Enveloper_0
  children [WavePlayer_5]
  properties {
   pitchCurve {editorCurveData {keys [{time=0}]}}
   volumeCurve {editorCurveData {keys [<at -0.0411 0.2084>,<at 0.2558 0.5977>,<at 1.1232 0.972>,<at 7.5 0.0495>]}}
  }
 }
 {id=WavePlayer_5, properties {soundWaveAssetPtr=/Figment_S07/Sounds/Fort_Impact_Sounds/Demolition/EarthQuake_Shakes_Med}}
]
properties {
 attenuationSettings=/Figment_S07/Sounds/Attenuation/Fort_Quest_Season7_3P
 bOverrideConcurrency=true
 maxDistance=13000
 soundClassObject=&Fort_UI
 volumeMultiplier=1
 concurrencyOverrides {maxCount=1, resolutionRule=StopOldest}
}
